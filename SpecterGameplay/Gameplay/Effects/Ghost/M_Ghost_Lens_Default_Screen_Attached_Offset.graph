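# Translucent unlit material for a ghost lens effect on Niagara sprites or mesh particles, drawn without depth testing as a screen-space overlay.
# A panning tiled pattern blends particle color toward a tint through an emissive adjustment function.
# An inverted radial gradient on viewport UVs concentrates opacity at screen edges, modulated by pattern, particle alpha and a UV wipe.
# Dynamic parameters drive wipe, pattern strength and radius; camera-minus-actor position math supplies the attachment offset.
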Material /SpecterGameplay/Gameplay/Effects/Ghost/M_Ghost_Lens_Default_Screen_Attached_Offset
graph {
 properties {
  bDisableDepthTest=true
  bUsedWithNiagaraMeshParticles=true
  bUsedWithNiagaraSprites=true
  blendMode=BLEND_Translucent
  pixelDepthOffsetMode=PDOM_Legacy
  shadingModel=MSM_Unlit
  usedShadingModels=Unlit
 }
 nodes [
  n1 = Time()
  n2 = Constant() {r=0.75}
  OneMinus(None=Frac(None=Multiply(A=@n1, B=@n2) {constB=0.5}))
  n3 = DynamicParameter() {defaultValue=<rgba 0.25 1 1.25 1>, paramNames [Wipe Mask, PatternTextureMult, Radius, Param4]}
  n4 = Multiply(A=@n3.Wipe Mask) {constB=1.5}
  OneMinus(None=@n4)
  n5 = ParticleColor()
  n6 = VectorParameter() {parameterName=ScrollingPatternColor, defaultValue=<rgba 0 2 1.8675 1>}
  n7 = ScalarParameter() {defaultValue=8, parameterName=PatternTileX}
  n8 = ScalarParameter() {defaultValue=8, parameterName=PatternTilieY}
  n9 = Panner(
   Coordinate=Multiply(A=TextureCoordinate(), B=AppendVector(A=@n7, B=@n8))
   Speed=AppendVector(A=ScalarParameter() {parameterName=PatternSpeedX}, B=ScalarParameter() {defaultValue=1.25, parameterName=PatternSpeedY})
  ) {speedY=1}
  n10 = Multiply(
   A=Power(Base=TextureSampleParameter2D(UVs=@n9) {parameterName=ScrollingPattern, texture=/Game/ContentCreationTools/ObjectPlacement/Textures/T_BlankWhite_01}.R)
   B=@n3.PatternTextureMult
  ) {constB=0.8}
  Reroute(None=MaterialFunctionCall(Emissive=LinearInterpolate(A=@n5.RGB, B=@n6.RGB, Alpha=@n10)) {materialFunction=&MF_FX_EmissiveAdjustments}.Result)
  n11 = OneMinus(None=MaterialFunctionCall(UVs=ScreenPosition().ViewportUV, Radius=@n3.Radius) {materialFunction=&RadialGradientExponential}.RadialGradientExponential)
  n12 = Multiply(A=Power(Base=@n11) {constExponent=12}) {constB=50}
  n13 = Multiply(A=Power(Base=@n11) {constExponent=5}) {constB=50}
  n14 = Add(A=Multiply(A=@n10, B=Saturate(None=@n12)), B=@n13)
  n15 = OneMinus(None=ComponentMask(None=TextureCoordinate()) {g=true})
  n16 = Saturate(None=Divide(A=Subtract(A=@n15, B=@n4), B=Constant() {desc=0.5, r=0.25}) {constB=0.5})
  Saturate(None=Multiply(A=@n5.A, B=Multiply(A=@n14, B=@n16)))
  Saturate(None=@n13)
  n17 = CameraPositionWS()
  n18 = WorldPosition()
  Subtract(A=Subtract(A=Add(A=@n17, B=@n18.XYZ), B=@n18.XYZ), B=ActorPositionWS())
 ]
}
