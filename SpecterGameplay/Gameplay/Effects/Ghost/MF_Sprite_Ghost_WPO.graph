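# MF_Sprite_Ghost_WPO is a material function that computes an animated wobble for ghost sprite world position offset.
# Two waves run along world Z and scroll with time at slightly different Speed-driven rates and different wavelengths: a cosine of a sine, plus a plain sine.
# Their sum is scaled by the Intensity parameter and masked by vertex color red, giving a single scalar output with no direction applied.
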MaterialFunction /SpecterGameplay/Gameplay/Effects/Ghost/MF_Sprite_Ghost_WPO
graph {nodes [
  n1 = WorldPosition()
  n2 = Time()
  n3 = ScalarParameter() {defaultValue=40, group=WPO, parameterName=Speed}
  n4 = Divide(A=Add(A=@n1.Z, B=Multiply(A=@n2, B=@n3) {constB=40})) {constB=200}
  n5 = Add(A=@n1.Z, B=Multiply(A=@n2, B=Multiply(A=@n3) {constB=0.9}) {constB=36})
  n6 = Add(A=Cosine(None=Sine(None=@n4)), B=Sine(None=Divide(A=@n5) {constB=150}))
  n7 = ScalarParameter() {defaultValue=5, group=WPO, parameterName=Intensity}
  FunctionOutput(None=Multiply(A=VertexColor().R, B=Multiply(A=@n6, B=@n7) {constB=4}))
 ]}
