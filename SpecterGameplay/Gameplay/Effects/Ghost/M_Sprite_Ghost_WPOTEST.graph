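# Test material for the Sprite Ghost effect, judging by its WPOTEST name.
# Its graph holds a single call to the MF_Sprite_Ghost_WPO material function, with no connections to material outputs shown.
# It uses Default Lit shading, previews on Sprite_Ghost_Creature, and enables skeletal mesh and mesh deformer usage with automatic usage flags turned off.
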Material /SpecterGameplay/Gameplay/Effects/Ghost/M_Sprite_Ghost_WPOTEST
graph {
 properties {
  bAutomaticallySetUsageInEditor=false
  bUsedWithMeshDeformer=true
  bUsedWithSkeletalMesh=true
  previewMesh=&/Sprite_Ghost_Creature
  usedShadingModels=Default Lit
 }
 nodes [MaterialFunctionCall() {materialFunction=&/MF_Sprite_Ghost_WPO}]
}
