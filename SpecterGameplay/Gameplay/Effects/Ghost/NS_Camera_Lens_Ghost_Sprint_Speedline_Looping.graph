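# Niagara camera-lens effect, named as the looping speedline overlay for the Ghost's sprint.
# The disabled AirMesh emitter continuously spawns faint purple mesh particles, whose alpha follows a fade-in/out curve scaled by System.VelocityAlpha.
# The enabled SpeedLines emitter lists no modules or renderers here.
# Six system-level variable assignments exist without visible values, and dependencies include vector-length, remap and color-array inputs.
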NiagaraSystem /SpecterGameplay/Gameplay/Effects/Ghost/NS_Camera_Lens_Ghost_Sprint_Speedline_Looping
dependencies {
 usedDataInterfaces [&NiagaraDataInterfaceCurve, &NiagaraDataInterfaceArrayColor]
 usedDynamicInputs [
  &Lerp_Float
  &RemapRange
  &VectorLength
  &MakeLinearColor
  &RandomRangeFloat
  &SelectLinearColorFromArray
  &MultiplyLinearColor
  &MakeVector4_FromLinearColor
 ]
 usedModules [
  &SystemState
  &self:SystemSpawnScript.SystemScriptSource.SystemScriptGraph.NiagaraNodeAssignment_0.SetVariables_#1
  &self:SystemSpawnScript.SystemScriptSource.SystemScriptGraph.NiagaraNodeAssignment_1.SetVariables_#2
  &self:SystemSpawnScript.SystemScriptSource.SystemScriptGraph.NiagaraNodeAssignment_2.SetVariables_#3
  &self:SystemSpawnScript.SystemScriptSource.SystemScriptGraph.NiagaraNodeAssignment_3.SetVariables_#4
  &self:SystemSpawnScript.SystemScriptSource.SystemScriptGraph.NiagaraNodeAssignment_4.SetVariables_#5
  &self:SystemSpawnScript.SystemScriptSource.SystemScriptGraph.NiagaraNodeAssignment_5.SetVariables_#6
 ]
}
emitters {
 AirMesh {
  bEnabled=false
  emitterUpdateScript [
   &EmitterState(Inactive Response=::NewEnumerator1, Life Cycle Mode=::NewEnumerator1, Loop Duration=1, UseLoopDelay=false)
   &SpawnBurst_Instantaneous(Use Spawn Probability=false, UseLoopCountLimit=false)
   &SpawnRate(SpawnRate=2, Use Spawn Probability=false)
  ]
  particleSpawnScript [
   &InitializeParticle(Color=<rgba 0.4738 0.0609 1 0.025>, UsePositionOffset=false, Write Material Random=true)
   &InitialMeshOrientation(Mesh Orientation Mode=::NewEnumerator7, Rotation=<0 0.5 0>, Use Rotation Vector=true)
  ]
  particleUpdateScript [
   &ParticleState()
   &ScaleColor(
    Scale Alpha=&FloatFromCurve(
     FloatCurve=DataInterface() {propertyValues {Curve {keys [
         {arriveTangent=5.8909, interpMode=RCIM_Cubic, leaveTangent=5.8909, tangentMode=RCTM_User, time=0, value=0}
         <at 0.3 1 RCIM_Cubic>
         {arriveTangent=-1.4286, interpMode=RCIM_Cubic, leaveTangent=-1.4286, time=1, value=0}
        ]}}}
     Scale Curve=&Multiply_Float(A=2, B=$System.VelocityAlpha)
    )
    ScaleA=true
    ScaleRGB=true
   )
  ]
  rendererClasses [&NiagaraMeshRendererProperties]
  renderers [{rendererClass=&NiagaraMeshRendererProperties, rendererIndex=0}]
 }
 SpeedLines {rendererClasses [], renderers []}
}
summary {
 systemName=NS_Camera_Lens_Ghost_Sprint_Speedline_Looping
 emitters [
  {bEnabled=false, emitterName=AirMesh, rendererClasses [&NiagaraMeshRendererProperties]}
  {emitterName=SpeedLines, rendererClasses []}
 ]
 userVariables []
}
